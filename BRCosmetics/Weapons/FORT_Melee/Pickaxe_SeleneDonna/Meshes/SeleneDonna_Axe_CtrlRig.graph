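# Control Rig for the SeleneDonna pickaxe that animates four balls circling two orbit rings around the axe head.
# Functions orbit 1 and orbit 2 each build and debug-draw a closed B-spline through eleven ring bones.
# A time loop continuously spins parent controls, carrying child slide controls.
# Each ball bone snaps to the spline point closest to its slide control, read before that evaluation's rotation update.
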ControlRigBlueprint /BRCosmetics/Weapons/FORT_Melee/Pickaxe_SeleneDonna/Meshes/SeleneDonna_Axe_CtrlRig
graphs [
 New Function_ContainedGraph {nodes [
   RigVMFunctionEntryNode()
   n1 = /Script/ControlRigSpline.RigUnit_ControlRigSplineFromPoints(
    Compression=0
    Points=((X=0,Y=0,Z=0),(X=0,Y=0,Z=0),(X=0,Y=0,Z=0),(X=0,Y=0,Z=0),(X=0,Y=0,Z=0),(X=0,Y=0,Z=0),(X=0,Y=0,Z=0),(X=0,Y=0,Z=0),(X=0,Y=0,Z=0),(X=0,Y=0,Z=0),(X=0,Y=0,Z=0))
    SamplesPerSegment=32
    SplineMode=BSpline
    Stretch=0
    bClosed=true
    Points.0=&RigUnit_GetTransform(Item=(Type=Bone,Name="for_orbit_2_0"), bInitial=False).Transform.Translation
    Points.1=&RigUnit_GetTransform(Item=(Type=Bone,Name="for_orbit_2_1"), bInitial=False).Transform.Translation
    Points.2=&RigUnit_GetTransform(Item=(Type=Bone,Name="for_orbit_2_2"), bInitial=False).Transform.Translation
    Points.3=&RigUnit_GetTransform(Item=(Type=Bone,Name="for_orbit_2_3"), bInitial=False).Transform.Translation
    Points.4=&RigUnit_GetTransform(Item=(Type=Bone,Name="for_orbit_2_4"), bInitial=False).Transform.Translation
    Points.5=&RigUnit_GetTransform(Item=(Type=Bone,Name="for_orbit_2_5"), bInitial=False).Transform.Translation
    Points.6=&RigUnit_GetTransform(Item=(Type=Bone,Name="for_orbit_2_6"), bInitial=False).Transform.Translation
    Points.7=&RigUnit_GetTransform(Item=(Type=Bone,Name="for_orbit_2_7"), bInitial=False).Transform.Translation
    Points.8=&RigUnit_GetTransform(Item=(Type=Bone,Name="for_orbit_2_8"), bInitial=False).Transform.Translation
    Points.9=&RigUnit_GetTransform(Item=(Type=Bone,Name="for_orbit_2_9"), bInitial=False).Transform.Translation
    Points.10=&RigUnit_GetTransform(Item=(Type=Bone,Name="for_orbit_2_10"), bInitial=False).Transform.Translation
   ) {is=struct RigVMUnitNode}
   > /Script/ControlRigSpline.RigUnit_DrawControlRigSpline(Color=(R=1,G=0,B=0,A=1), Detail=32, Thickness=0.2, Spline=@n1.Spline) {is=struct RigVMUnitNode}
   n2 = &RigUnit_GetTransform(Item=(Type=Control,Name="slide_on_curve_ball_0_orbit_2_ctrl"), bInitial=False)
   > RigUnit_SetTranslation = /Script/ControlRig.RigUnit_SetTranslation(
    Item=(Type=Bone,Name="dyn_ball_0_orbit_2")
    Space=GlobalSpace
    Weight=1
    bInitial=False
    bPropagateToChildren=true
    Value=/Script/ControlRigSpline.RigUnit_TransformFromControlRigSpline(
     Roll=0
     UpVector=(X=0,Y=0,Z=1)
     U=/Script/ControlRigSpline.RigUnit_ClosestParameterFromControlRigSpline(Spline=@n1.Spline, Position=@n2.Transform.Translation) {is=struct RigVMUnitNode}.U
     Spline=@n1.Spline
    ) {is=struct RigVMUnitNode}.Transform.Translation
   ) {is=struct RigVMUnitNode, template="Set Transform::Execute(in Item,in Space,in bInitial,in Value,in Weight,in bPropagateToChildren)"}
   n3 = &RigUnit_GetTransform(Item=(Type=Control,Name="slide_on_curve_ball_1_orbit_2_ctrl"), bInitial=False)
   > &RigUnit_SetTranslation(
    Item=(Type=Bone,Name="dyn_ball_1_orbit_2")
    Space=GlobalSpace
    Weight=1
    bInitial=False
    bPropagateToChildren=True
    Value=&RigUnit_TransformFromControlRigSpline(
     Roll=0
     UpVector=(X=0,Y=0,Z=1)
     U=&RigUnit_ClosestParameterFromControlRigSpline(Spline=@n1.Spline, Position=@n3.Transform.Translation).U
     Spline=@n1.Spline
    ).Transform.Translation
   ) {template="Set Transform::Execute(in Item,in Space,in bInitial,in Value,in Weight,in bPropagateToChildren)"}
   RigVMFunction_TimeLoop = /Script/RigVM.RigVMFunction_TimeLoop(Duration=360, Normalize=false, Speed=100) {is=struct RigVMUnitNode}
   RigVMFunction_MathQuaternionFromEuler = &RigVMFunction_MathQuaternionFromEuler(Euler=(X=0,Y=0,Z=0), Euler.X=@RigVMFunction_TimeLoop.Relative)
   > RigUnit_SetRotation = &RigUnit_SetRotation(Item=(Type=Control,Name="dyn_main_rot_ball_0_orbit_2_ctrl"), Weight=1, bInitial=false, Value=@RigVMFunction_MathQuaternionFromEuler.Result)
   > &RigUnit_SetRotation(Item=(Type=Control,Name="dyn_main_rot_ball_1_orbit_2_ctrl"), Weight=1, bInitial=false, Value=@RigVMFunction_MathQuaternionFromEuler.Result)
   > RigVMFunctionReturnNode()
  ]}
 New Function_ContainedGraph {nodes [
   RigVMFunctionEntryNode()
   n1 = &RigUnit_ControlRigSplineFromPoints(
    Compression=0
    Points=((X=0,Y=0,Z=0),(X=0,Y=0,Z=0),(X=0,Y=0,Z=0),(X=0,Y=0,Z=0),(X=0,Y=0,Z=0),(X=0,Y=0,Z=0),(X=0,Y=0,Z=0),(X=0,Y=0,Z=0),(X=0,Y=0,Z=0),(X=0,Y=0,Z=0),(X=0,Y=0,Z=0))
    SamplesPerSegment=32
    SplineMode=BSpline
    Stretch=0
    bClosed=true
    Points.0=&RigUnit_GetTransform(Item=(Type=Bone,Name="for_orbit_1_0"), bInitial=False).Transform.Translation
    Points.1=&RigUnit_GetTransform(Item=(Type=Bone,Name="for_orbit_1_1"), bInitial=False).Transform.Translation
    Points.2=&RigUnit_GetTransform(Item=(Type=Bone,Name="for_orbit_1_2"), bInitial=False).Transform.Translation
    Points.3=&RigUnit_GetTransform(Item=(Type=Bone,Name="for_orbit_1_3"), bInitial=False).Transform.Translation
    Points.4=&RigUnit_GetTransform(Item=(Type=Bone,Name="for_orbit_1_4"), bInitial=False).Transform.Translation
    Points.5=&RigUnit_GetTransform(Item=(Type=Bone,Name="for_orbit_1_5"), bInitial=False).Transform.Translation
    Points.6=&RigUnit_GetTransform(Item=(Type=Bone,Name="for_orbit_1_6"), bInitial=False).Transform.Translation
    Points.7=&RigUnit_GetTransform(Item=(Type=Bone,Name="for_orbit_1_7"), bInitial=False).Transform.Translation
    Points.8=&RigUnit_GetTransform(Item=(Type=Bone,Name="for_orbit_1_8"), bInitial=False).Transform.Translation
    Points.9=&RigUnit_GetTransform(Item=(Type=Bone,Name="for_orbit_1_9"), bInitial=False).Transform.Translation
    Points.10=&RigUnit_GetTransform(Item=(Type=Bone,Name="for_orbit_1_10"), bInitial=False).Transform.Translation
   )
   > &RigUnit_DrawControlRigSpline(Color=(R=1,G=0.1409,B=0.9615,A=1), Detail=32, Thickness=0.2, Spline=@n1.Spline)
   n2 = &RigUnit_GetTransform(Item=(Type=Control,Name="slide_on_curve_ball_0_orbit_1_ctrl"), bInitial=False)
   > RigUnit_SetTranslation = &RigUnit_SetTranslation(
    Item=(Type=Bone,Name="dyn_ball_0_orbit_1")
    Space=GlobalSpace
    Weight=1
    bInitial=False
    bPropagateToChildren=True
    Value=&RigUnit_TransformFromControlRigSpline(
     Roll=0
     UpVector=(X=0,Y=0,Z=1)
     U=&RigUnit_ClosestParameterFromControlRigSpline(Spline=@n1.Spline, Position=@n2.Transform.Translation).U
     Spline=@n1.Spline
    ).Transform.Translation
   ) {template="Set Transform::Execute(in Item,in Space,in bInitial,in Value,in Weight,in bPropagateToChildren)"}
   RigUnit_GetTransform = &RigUnit_GetTransform(Item=(Type=Control,Name="slide_on_curve_ball_1_orbit_1_ctrl"), bInitial=False)
   > &RigUnit_SetTranslation(
    Item=(Type=Bone,Name="dyn_ball_1_orbit_1")
    Space=GlobalSpace
    Weight=1
    bInitial=False
    bPropagateToChildren=True
    Value=&RigUnit_TransformFromControlRigSpline(
     Roll=0
     UpVector=(X=0,Y=0,Z=1)
     U=&RigUnit_ClosestParameterFromControlRigSpline(Spline=@n1.Spline, Position=@RigUnit_GetTransform.Transform.Translation).U
     Spline=@n1.Spline
    ).Transform.Translation
   ) {template="Set Transform::Execute(in Item,in Space,in bInitial,in Value,in Weight,in bPropagateToChildren)"}
   RigVMFunction_TimeLoop = &RigVMFunction_TimeLoop(Duration=360, Normalize=false, Speed=100)
   RigVMFunction_MathQuaternionFromEuler = &RigVMFunction_MathQuaternionFromEuler(Euler=(X=0,Y=0,Z=0), Euler.X=@RigVMFunction_TimeLoop.Relative)
   > RigUnit_SetRotation = &RigUnit_SetRotation(Item=(Type=Control,Name="dyn_main_rot_ball_0_orbit_1_ctrl"), Weight=1, bInitial=false, Value=@RigVMFunction_MathQuaternionFromEuler.Result)
   > &RigUnit_SetRotation(Item=(Type=Control,Name="dyn_main_rot_ball_1_orbit_1_ctrl"), Weight=1, bInitial=false, Value=@RigVMFunction_MathQuaternionFromEuler.Result)
   > RigVMFunctionReturnNode()
  ]}
 RigVMFunctionLibrary {nodes [
   orbit 1 = RigVMCollapseNode(ExecuteContext=())
   orbit 2 = RigVMCollapseNode(ExecuteContext=())
  ]}
 RigVMModel {nodes [
   RigUnit_BeginExecution = &RigUnit_BeginExecution()
   orbit 1 = RigVMFunctionReferenceNode(ExecuteContext=@RigUnit_BeginExecution.ExecutePin)
   > orbit 2 = RigVMFunctionReferenceNode()
  ]}
]
hierarchy [
 b_AR_Root {loc=<0 0 0>, rot=<0 0 0.7071 0.7071>}
 Melee_TwoHanded_Handle {parent=b_AR_Root, loc=<8.1477 -0.008 -29.0793>, rot=<0 0 0.7071 -0.7071>}
 base_main_orbit_2 {parent=Melee_TwoHanded_Handle, loc=<2.3551 -0.2814 81.5012>, rot=<0.0086 -0.824 -0.3163 0.4699>}
 dyn_main_orbit_2 {parent=base_main_orbit_2, loc=<0 0 0>, rot=<0 0 0 1>}
 base_ball_0_orbit_2 {parent=dyn_main_orbit_2, loc=<-3.7264 8.3399 -16.6646>, rot=<0.6351 0.5831 0.4081 0.3002>}
 dyn_ball_0_orbit_2 {parent=base_ball_0_orbit_2, loc=<0 0 0>, rot=<0 0 0 1>}
 base_ball_1_orbit_2 {parent=dyn_main_orbit_2, loc=<1.438 -25.3182 2.5143>, rot=<0.2461 -0.6812 -0.1899 -0.6629>}
 dyn_ball_1_orbit_2 {parent=base_ball_1_orbit_2, loc=<0 0 0>, rot=<0 0 0 1>}
 for_orbit_2_0 {parent=dyn_main_orbit_2, loc=<-0.5511 27.7838 1.2714>, rot=<0.0086 -0.824 -0.3163 -0.4699>}
 for_orbit_2_1 {parent=dyn_main_orbit_2, loc=<1.9712 19.3041 11.9914>, rot=<0.0086 -0.824 -0.3163 -0.4699>}
 for_orbit_2_2 {parent=dyn_main_orbit_2, loc=<3.4234 7.8193 17.8109>, rot=<0.0086 -0.824 -0.3163 -0.4699>}
 for_orbit_2_3 {parent=dyn_main_orbit_2, loc=<4.1449 -5.3154 19.0687>, rot=<0.0086 -0.824 -0.3163 -0.4699>}
 for_orbit_2_4 {parent=dyn_main_orbit_2, loc=<4.1539 -18.2542 17.648>, rot=<0.0086 -0.824 -0.3163 -0.4699>}
 for_orbit_2_5 {parent=dyn_main_orbit_2, loc=<2.3037 -28.3568 7.0884>, rot=<0.0086 -0.824 -0.3163 -0.4699>}
 for_orbit_2_6 {parent=dyn_main_orbit_2, loc=<-0.4562 -24.5805 -6.3013>, rot=<0.0086 -0.824 -0.3163 -0.4699>}
 for_orbit_2_7 {parent=dyn_main_orbit_2, loc=<-2.4327 -14.16 -14.3316>, rot=<0.0086 -0.824 -0.3163 -0.4699>}
 for_orbit_2_8 {parent=dyn_main_orbit_2, loc=<-3.568 -1.6075 -17.8208>, rot=<0.0086 -0.824 -0.3163 -0.4699>}
 for_orbit_2_9 {parent=dyn_main_orbit_2, loc=<-3.9424 11.3338 -17.7426>, rot=<0.0086 -0.824 -0.3163 -0.4699>}
 for_orbit_2_10 {parent=dyn_main_orbit_2, loc=<-3.3079 24.1746 -12.5966>, rot=<0.0086 -0.824 -0.3163 -0.4699>}
 base_main_rot_ball {parent=dyn_main_orbit_2, loc=<0 0 0>, rot=<0 0 0 1>}
 dyn_main_rot_ball {parent=base_main_rot_ball, loc=<0 0 0>, rot=<0 0 0 1>}
 slide_on_curve {parent=dyn_main_rot_ball, loc=<0.773 28.3392 1.4286>, rot=<0 0 0 1>}
 base_main_orbit_1 {parent=Melee_TwoHanded_Handle, loc=<1.3618 1.4438 81.5958>, rot=<0.1816 -0.6782 0.3779 -0.6035>}
 dyn_main_orbit_1 {parent=base_main_orbit_1, loc=<0 0 0>, rot=<0 0 0 1>}
 base_ball_0_orbit_1 {parent=dyn_main_orbit_1, loc=<-0.1511 22.4931 -0.9155>, rot=<-0.309 0.4961 -0.2652 -0.7669>}
 dyn_ball_0_orbit_1 {parent=base_ball_0_orbit_1, loc=<0 0 0>, rot=<0 0 0 1>}
 base_ball_1_orbit_1 {parent=dyn_main_orbit_1, loc=<-2.1524 -19.9151 -4.202>, rot=<-0.4899 0.6068 0.0429 -0.6245>}
 dyn_ball_1_orbit_1 {parent=base_ball_1_orbit_1, loc=<0 0 0>, rot=<0 0 0 1>}
 for_orbit_1_0 {parent=dyn_main_orbit_1, loc=<-2.9878 -21.6445 -6.3527>, rot=<-0.1816 0.6782 -0.3779 -0.6035>}
 for_orbit_1_1 {parent=dyn_main_orbit_1, loc=<1.2283 -24.9527 4.6633>, rot=<-0.1816 0.6782 -0.3779 -0.6035>}
 for_orbit_1_2 {parent=dyn_main_orbit_1, loc=<3.8743 -14.6084 11.5092>, rot=<-0.1816 0.6782 -0.3779 -0.6035>}
 for_orbit_1_3 {parent=dyn_main_orbit_1, loc=<4.5737 -5.2483 13.6063>, rot=<-0.1816 0.6782 -0.3779 -0.6035>}
 for_orbit_1_4 {parent=dyn_main_orbit_1, loc=<4.4838 6.457 13.2421>, rot=<-0.1816 0.6782 -0.3779 -0.6035>}
 for_orbit_1_5 {parent=dyn_main_orbit_1, loc=<3.6656 16.3254 11.3432>, rot=<-0.1816 0.6782 -0.3779 -0.6035>}
 for_orbit_1_6 {parent=dyn_main_orbit_1, loc=<0.945 25.9412 3.6894>, rot=<-0.1816 0.6782 -0.3779 -0.6035>}
 for_orbit_1_7 {parent=dyn_main_orbit_1, loc=<-3.3243 21.3164 -7.2356>, rot=<-0.1816 0.6782 -0.3779 -0.6035>}
 for_orbit_1_8 {parent=dyn_main_orbit_1, loc=<-4.5514 10.8878 -10.7094>, rot=<-0.1816 0.6782 -0.3779 -0.6035>}
 for_orbit_1_9 {parent=dyn_main_orbit_1, loc=<-4.9457 -0.1435 -11.5579>, rot=<-0.1816 0.6782 -0.3779 -0.6035>}
 for_orbit_1_10 {parent=dyn_main_orbit_1, loc=<-4.3988 -10.2107 -10.3986>, rot=<-0.1816 0.6782 -0.3779 -0.6035>}
 base_main_rot_ball_orbit_1 {parent=base_main_orbit_1, loc=<0 0 0>, rot=<0.0678 -0.1732 -0.012 -0.9825>}
 dyn_main_rot_ball_orbit_1 {parent=base_main_rot_ball_orbit_1, loc=<0 0 0>, rot=<0 0 0 1>}
 slide_on_curve_orbit_1 {parent=dyn_main_rot_ball_orbit_1, loc=<0 -25.68 0>, rot=<0 0 0 1>}
 base_star {parent=Melee_TwoHanded_Handle, loc=<1.5705 0.1087 82.6684>, rot=<0 -0.7071 0 -0.7071>}
 dyn_star {parent=base_star, loc=<0 0 0>, rot=<0 0 0 1>}
 dyn_main_rot_ball_0_orbit_2_ctrl_offset {controlType=EULER_TRANSFORM, type=Control, loc=<0 0 0>, rot=<0 0 0 1>}
 dyn_main_rot_ball_0_orbit_2_ctrl {controlType=EULER_TRANSFORM, parent=dyn_main_rot_ball_0_orbit_2_ctrl_offset, type=Control, loc=<0 0 0>, rot=<0 0 0 1>}
 slide_on_curve_ball_0_orbit_2_ctrl {controlType=EULER_TRANSFORM, parent=dyn_main_rot_ball_0_orbit_2_ctrl, type=Control, loc=<0 0 0>, rot=<0 0 0 1>}
 dyn_main_rot_ball_0_orbit_1_ctrl_offset {controlType=EULER_TRANSFORM, type=Control, loc=<0 0 0>, rot=<0 0 0 1>}
 dyn_main_rot_ball_0_orbit_1_ctrl {controlType=EULER_TRANSFORM, parent=dyn_main_rot_ball_0_orbit_1_ctrl_offset, type=Control, loc=<0 0 0>, rot=<0 0 0 1>}
 slide_on_curve_ball_0_orbit_1_ctrl {controlType=EULER_TRANSFORM, parent=dyn_main_rot_ball_0_orbit_1_ctrl, type=Control, loc=<0 0 0>, rot=<0 0 0 1>}
 dyn_main_rot_ball_1_orbit_2_ctrl_offset {controlType=EULER_TRANSFORM, type=Control, loc=<0 0 0>, rot=<0 0 0 1>}
 dyn_main_rot_ball_1_orbit_2_ctrl {controlType=EULER_TRANSFORM, parent=dyn_main_rot_ball_1_orbit_2_ctrl_offset, type=Control, loc=<0 0 0>, rot=<0 0 0 1>}
 slide_on_curve_ball_1_orbit_2_ctrl {controlType=EULER_TRANSFORM, parent=dyn_main_rot_ball_1_orbit_2_ctrl, type=Control, loc=<0 0 0>, rot=<0 0 0 1>}
 dyn_main_rot_ball_1_orbit_1_ctrl_offset {controlType=EULER_TRANSFORM, type=Control, loc=<0 0 0>, rot=<0 0 0 1>}
 dyn_main_rot_ball_1_orbit_1_ctrl {controlType=EULER_TRANSFORM, parent=dyn_main_rot_ball_1_orbit_1_ctrl_offset, type=Control, loc=<0 0 0>, rot=<0 0 0 1>}
 slide_on_curve_ball_1_orbit_1_ctrl {controlType=EULER_TRANSFORM, parent=dyn_main_rot_ball_1_orbit_1_ctrl, type=Control, loc=<0 0 0>, rot=<0 0 0 1>}
]
