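# Animation Blueprint for the SeleneDonna pickaxe that adds procedural secondary motion to its star and orbit bones.
# The input pose runs through a Control Rig, an additive star bone modifier, then chained Anim Dynamics nodes with angular springs, limits and reduced gravity for dyn_star and both orbit bones.
# Each update computes a gravity override through a deprecated helper, plus real-time sine-wave translation and rotation values for the star.
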AnimBlueprint /BRCosmetics/Weapons/FORT_Melee/Pickaxe_SeleneDonna/Meshes/SeleneDonna_Axe_AnimBP
target_skeleton=&/SeleneDonna_Axe_Skeleton
graphs [
 AnimGraph {nodes [
   AnimGraphNode_SubInput_0 = Input Pose()
   n1 = Control Rig(Source=Component To Local(ComponentPose=Local To Component(LocalPose=@AnimGraphNode_SubInput_0))) {
    alphaBoolBlend {blendOption=Linear}
    controlRigAssetReference {blueprintRigClass=&/SeleneDonna_Axe_CtrlRig.SeleneDonna_Axe_CtrlRig_C}
   }
   n2 = "Transform (Modify) Bone - Bone: dyn_star"!(ComponentPose=Local To Component(LocalPose=@n1), Alpha=1, Translation="0,0,0", Rotation="0,0,0") -> (Pose) {
    rotationMode=BMM_Additive
    rotationSpace=BCS_BoneSpace
    translationMode=BMM_Additive
    translationSpace=BCS_BoneSpace
    alphaBoolBlend {blendOption=Linear}
    boneToModify {boneName=dyn_star}
   }
   n3 = Anim Dynamics - Bone: dyn_star!(ComponentPose=@n2, Alpha=1, GravityOverride:Vector="0,0,-980") -> (Pose) {
    angularDampingOverride=1
    angularSpringConstant=20000
    bAngularSpring=true
    bGravityOverrideInSimSpace=true
    bOverrideAngularDamping=true
    bUseGravityOverride=true
    gravityScale=0.35
    simulationSpace=RootRelative
    alphaBoolBlend {blendOption=Linear}
    boundBone {boneName=dyn_star}
    componentAppliedLinearAccClamp=<3000 3000 3000>
    componentLinearAccScale=<0.5 0.5 0.5>
    componentLinearVelScale=<0.5 0.5 0.5>
    gravityOverride {z=-980}
    physicsBodyDefinitions [{
      boundBone {boneName=dyn_star}
      boxExtents=<80 80 80>
      constraintSetup {angularLimitsMax=<0.3 0.3 0.3>, angularLimitsMin=<-0.3 -0.3 -0.3>, angularTarget=<1 0 0>}
      localJointOffset=<8 0 0>
     }]
   }
   n4 = Anim Dynamics - Bone: dyn_star(ComponentPose=@n3, Alpha=1, GravityOverride="0,0,-980") {
    angularDampingOverride=1
    angularSpringConstant=20000
    bAngularSpring=true
    bGravityOverrideInSimSpace=true
    bOverrideAngularDamping=true
    bUseGravityOverride=true
    gravityScale=0.35
    simulationSpace=RootRelative
    alphaBoolBlend {blendOption=Linear}
    boundBone {boneName=dyn_star}
    componentAppliedLinearAccClamp=<3000 3000 3000>
    componentLinearAccScale=<0.5 0.5 0.5>
    componentLinearVelScale=<0.5 0.5 0.5>
    gravityOverride {z=-980}
    physicsBodyDefinitions [{
      boundBone {boneName=dyn_star}
      boxExtents=<75 75 75>
      constraintSetup {angularLimitsMax=<5 5 5>, angularLimitsMin=<-5 -5 -5>, angularTarget=<1 0 0>}
      localJointOffset=<6 0 0>
     }]
   }
   n5 = Anim Dynamics - Bone: dyn_main_orbit_2(ComponentPose=@n4, Alpha=1, ExternalForce="0,0,0") {
    angularDampingOverride=1
    angularSpringConstant=20000
    bAngularSpring=true
    bGravityOverrideInSimSpace=true
    bOverrideAngularDamping=true
    bUseGravityOverride=true
    gravityScale=0.35
    simulationSpace=RootRelative
    alphaBoolBlend {blendOption=Linear}
    boundBone {boneName=dyn_main_orbit_2}
    componentAppliedLinearAccClamp=<3000 3000 3000>
    componentLinearAccScale=<0.5 0.5 0.5>
    componentLinearVelScale=<0.5 0.5 0.5>
    gravityOverride {z=-980}
    physicsBodyDefinitions [{
      boundBone {boneName=dyn_main_orbit_2}
      boxExtents=<50 50 50>
      constraintSetup {angularLimitsMax=<0 0 8>, angularLimitsMin=<0 0 -8>, angularTarget=<1 0 0>}
      localJointOffset=<4 0 0>
     }]
   }
   n6 = Anim Dynamics - Bone: dyn_main_orbit_1(ComponentPose=@n5, Alpha=1, ExternalForce="0,0,0") {
    angularDampingOverride=1
    angularSpringConstant=25000
    bAngularSpring=true
    bGravityOverrideInSimSpace=true
    bOverrideAngularDamping=true
    bUseGravityOverride=true
    gravityScale=0.35
    simulationSpace=RootRelative
    alphaBoolBlend {blendOption=Linear}
    boundBone {boneName=dyn_main_orbit_1}
    componentAppliedLinearAccClamp=<3000 3000 3000>
    componentLinearAccScale=<0.6 0.6 0.6>
    componentLinearVelScale=<0.6 0.6 0.6>
    gravityOverride {z=-980}
    physicsBodyDefinitions [{
      boundBone {boneName=dyn_main_orbit_1}
      boxExtents=<50 50 50>
      constraintSetup {angularLimitsMax=<0 0 20>, angularLimitsMin=<0 0 -20>, angularTarget=<1 0 0>}
      localJointOffset=<4 8 0>
     }]
   }
   Output Pose(Result=Component To Local(ComponentPose=@n6)) {layerGroup=None}
  ]}
 EventGraph {nodes [
   Event BlueprintUpdateAnimation()
   > n1 = Gravity Override State Function DEPRECATED(CharacterPartAnimInstance=Self-Reference(), GameTimeInSeconds=GetGameTimeInSeconds(), GravityOverrideParams=Get Gravity Override Params())
   > Set GravityOverride(GravityOverride=@n1)
   n2 = GetRealTimeSeconds()
   n3 = "Sin (Radians)"(A=float * float(A=@n2, B=1.8))
   > Set Translate Star(Translate Star:Vector=MakeVector(X=float * float(A=@n3, B=1.5), Y=float / float(A=@n3, B=0.8) {is=title K2Node_CallFunction}, Z=float / float(A=@n3, B=2))) -> (Output_Get:Vector)
   n4 = "Sin (Radians)"(A=float * float(A=@n2, B=0.8))
   n5 = float * float(A=@n4, B=5)
   > Set Rotate Star(Rotate Star:Rotator=MakeRotator(Roll=float * float(A=@n4, B=2), Pitch=@n5, Yaw=@n5)) -> (Output_Get:Rotator)
   TryGetPawnOwner()
  ]}
]
variables [
 Gravity Override Params {type=GravityOverrideParamsStruct}
 GravityOverride {type=Vector}
 Rotate Star {type=Rotator}
 Translate Star {type=Vector}
]
